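; Parent Blueprint for Chapter 5 light-source props, with optional time-of-day lighting and UEFN asset exposure.
; Construction toggles the TOD lighting component's auto-activation before the parent call, then pushes an emissive color multiplier into the mesh's dynamic materials.
; Time-of-day defaults take volumetric scattering from the first light component and the shadow-casting setting from a variable.
; Asset collection allows mesh materials and their Nanite overrides to be referenced in UEFN.
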(event EventBeginPlay
  (|CallParentFunction))

(event Collision|EventActorBeginOverlap
  (params (OtherActor :type Actor%20Object%20Reference)))

(event EventTick
  (params (DeltaSeconds :type Float%20%28single-precision%29)))

(fn BuildExposedAssetProviderVersion(Implementation)
  (params (Version :type Fort%20Exposed%20Asset%20Provider%20Version%20Structure%20%28by%20ref%29))
  (|CallParentFunction Version)
  (ExposedAssetProvider|AddVersion Version 2)
  (return))

(fn CollectAssetReferencestoAllow(Implementation)
  (params (Collector :type Fort%20Exposed%20Asset%20Reference%20Collector%20Structure%20%28by%20ref%29))
  ;[1 1 1] Collect the materials to allow them to be referenced in UEFN
  (|CallParentFunction Collector)
  (for _array_element (Rendering|Material|GetMaterials (Variables|Debug|GetStaticMeshComponent))
    (ExposedAssetProvider|CollectIndirectAssetReference Collector (Utilities|ResolveSoftReference _array_element))
    (if (Utilities|IsValid (Rendering|Material|GetNaniteOverideMaterial _array_element) :!class /Script/Engine.KismetSystemLibrary)
      (ExposedAssetProvider|CollectIndirectAssetReference Collector (Utilities|ResolveSoftReference (Rendering|Material|GetNaniteOverideMaterial _array_element)))))
  (return))

(fn GetTimeOfDayBlueprintDefaultVariables
  (bind _returnvalue (Actor|GetComponentsByClass "/Script/Engine.LightComponent"))
  (bind _output (Utilities|Array|Get(acopy) _returnvalue))
  (if (not (Utilities|Array|IsEmpty _returnvalue))
    (Utilities|IsValid _output
      (:Is%20Valid
        (return (Utilities|Struct|MakeTimeOfDayBlueprintDefaultVariables 0 (Variables|/Script/Engine.LightComponentBase|GetVolumetricScatteringIntensity _output) false (Variables|00_GlobalLighting|GetDisable%20Static%20Mesh%20Shadow%20Casting) false true)))
      (:Is%20Not%20Valid
        _)))
  (return (Utilities|Struct|MakeTimeOfDayBlueprintDefaultVariables 0 0.0 false false false true)))

(fn ConstructionScript
  (bind _staticmeshcomponent (Variables|Debug|GetStaticMeshComponent))
  (bind _returnvalue (Rendering|Material|GetNumMaterials _staticmeshcomponent))
  (bind _emissive_color_modifier (Variables|00_GlobalLighting|GetEmissive%20Color%20Modifier))
  (bind _todlightingcomponent (Variables|Default|GetTODLightingComponent))
  (if (Variables|00_GlobalLighting|GetTimeOfDayLight)
    ;[1 1 1] is the Light a TOD light?
    (Components|Activation|SetAutoActivate _todlightingcomponent true)
    (Class|TODLightingComponent|RunConstruct (Variables|Default|GetTODLightingComponent) (Variables|00_GlobalLighting|GetTest%20Day%20Phase))
    (else
      (Components|Activation|SetAutoActivate _todlightingcomponent false)))
  ;[1 1 1] Must run after setting AutoActivate
  (|CallParentFunction)
  ;[1 1 1] Multiply Emissive Color on Mesh via Emissive Color Modifier
  (if (> _returnvalue 0)
    (for _index (range _returnvalue)
      (Rendering|Material|SetVectorParameterValue (Class|BTechArtGlobalLibrary|GetDynamicMaterialInstance _staticmeshcomponent _index) "Emissive_Color_Multiplier" _emissive_color_modifier :!class /Script/Engine.MaterialInstanceDynamic))
    (else
      (Rendering|Material|SetVectorParameterValue (Class|BTechArtGlobalLibrary|GetDynamicMaterialInstance _staticmeshcomponent) "Emissive_Color_Multiplier" _emissive_color_modifier :!class /Script/Engine.MaterialInstanceDynamic))))
