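; Point light variant of the Chapter 5 global prop light source, extending a parent Blueprint with time-of-day light settings.
; Its construction script runs the parent, then, where cosmetic events can execute and randomization is enabled, seeds a random stream from actor location.
; The resulting float feeds "RandomLightval" on a dynamic instance of the point light's light function material.
; This graph never assigns that instance back to the light.
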(event EventBeginPlay
  (|CallParentFunction) ; This node is disabled and will not be called.\nDrag off pins to build functionality.
)

(event Collision|EventActorBeginOverlap
  (params (OtherActor :type Actor%20Object%20Reference))
  (|CallParentFunction OtherActor) ; This node is disabled and will not be called.\nDrag off pins to build functionality.
)

(event EventTick
  (params (DeltaSeconds :type Float%20%28single-precision%29))
  (|CallParentFunction DeltaSeconds) ; This node is disabled and will not be called.\nDrag off pins to build functionality.
)

(fn GetTimeOfDayBlueprintDefaultVariables
  (bind _outvariables (|CallParentFunction))
  (return _outvariables))

(fn ConstructionScript
  (|CallParentFunction)
  (Networking|CanExecuteCosmeticEvents
    (:True
      (if (Variables|Building|TimeOfDayLights|GetCreateRandomValueForLightFunction)
        (bind _output_get (Variables|Default|SetInstanceRandomStream (Math|Random|MakeRandomStreamfromLocation (Transformation|GetActorLocation) 1.0 true)))
        (Variables|Default|SetRandomFloat (Math|Random|RandomFloatfromStream _output_get))
        (bind _returnvalue (Rendering|Material|CreateDynamicMaterialInstance (Variables|/Script/Engine.LightComponent|GetLightFunctionMaterial (Variables|Default|GetPointLight)) "None" "None" :!class /Script/Engine.KismetMaterialLibrary))
        (Rendering|Material|SetScalarParameterValue _returnvalue "RandomLightval" (Variables|Default|GetRandomFloat) :!class /Script/Engine.MaterialInstanceDynamic)))))
