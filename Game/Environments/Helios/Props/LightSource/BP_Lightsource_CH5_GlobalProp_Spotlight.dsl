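; BP_Lightsource_CH5_GlobalProp_Spotlight is a spotlight prop Blueprint that can randomize its light function per placed instance.
; After the parent construction script, an optional flag derives Random Float from the actor's location through EnvironmentArtFuncLib.
; If the spotlight has a light function material, it creates a dynamic instance and sets RandomLightval, but never assigns that instance back to the light.
; BeginPlay, overlap and Tick are disabled stubs.
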(event EventBeginPlay
  (|CallParentFunction) ; This node is disabled and will not be called.\nDrag off pins to build functionality.
)

(event Collision|EventActorBeginOverlap
  (params (OtherActor :type Actor%20Object%20Reference))
  (|CallParentFunction OtherActor) ; This node is disabled and will not be called.\nDrag off pins to build functionality.
)

(event EventTick
  (params (DeltaSeconds :type Float%20%28single-precision%29))
  (|CallParentFunction DeltaSeconds) ; This node is disabled and will not be called.\nDrag off pins to build functionality.
)

(fn ConstructionScript
  (bind _spotlight (Variables|Default|GetSpotLight))
  (bind _lightfunctionmaterial (Variables|/Script/Engine.LightComponent|GetLightFunctionMaterial _spotlight))
  (|CallParentFunction)
  (if (Variables|Default|GetCreate%20Random%20Value%20for%20Light%20Function)
    (Variables|Default|SetRandom%20Float (Class|EnvironmentArtFuncLib|RandomFloatbyLocation (Transformation|GetActorLocation) 64.0 true))
    (if (Utilities|IsValid _lightfunctionmaterial :!class /Script/Engine.KismetSystemLibrary)
      (bind _returnvalue (Rendering|Material|CreateDynamicMaterialInstance _lightfunctionmaterial "None" "None" :!class /Script/Engine.KismetMaterialLibrary))
      (Rendering|Material|SetScalarParameterValue _returnvalue "RandomLightval" (Variables|Default|GetRandom%20Float) :!class /Script/Engine.MaterialInstanceDynamic))))
